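# Translucent unlit glow material for the TungStan pickaxe FX, built for Niagara mesh particles.
# Particle color feeds emissive through MF_FX_EmissiveAdjustments, while particle alpha scales a Fresnel rim term.
# Panning Perlin noise distorts the UVs of a second noise sample, which is sharpened into hot spots.
# The rim mask combines a steady base glow with those animated spots, brightening the edges unevenly.
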Material /BRCosmetics/Weapons/FORT_Melee/Pickaxe_TungStan/FX/M_Pickaxe_TungStan_Glow
graph {
 properties {
  bUsedWithNiagaraMeshParticles=true
  blendMode=BLEND_Translucent
  pixelDepthOffsetMode=PDOM_Legacy
  shadingModel=MSM_Unlit
  usedShadingModels=Unlit
 }
 nodes [
  n1 = ParticleColor()
  n2 = DotProduct(A=VertexNormalWS(), B=CameraVectorWS())
  n3 = Multiply(A=@n1.A, B=Power(Base=OneMinus(None=@n2)) {constExponent=3})
  n4 = Multiply(A=TextureCoordinate()) {constB=1.5}
  n5 = Panner() {speedX=0.5, speedY=0.2}
  n6 = Multiply(A=ComponentMask(None=TextureSample(UVs=@n5) {texture=/Game/Global/Textures/Default/Noise/3xPerlin_M}.RGB) {g=true, r=true}) {constB=0.5}
  n7 = ComponentMask(None=TextureSample(UVs=Add(A=@n4, B=@n6)) {texture=&3xPerlin_M}.B) {r=true}
  n8 = Saturate(None=Power(Base=Multiply(A=@n7) {constB=2}) {constExponent=10})
  Add(A=Multiply(A=@n3) {constB=0.85}, B=Multiply(A=@n3, B=Multiply(A=@n8) {constB=3}))
  MaterialFunctionCall(Emissive=@n1.RGB) {materialFunction=&MF_FX_EmissiveAdjustments}
 ]
}
