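# Unlit, translucent, two-sided particle material for the TungStan pickaxe FX, usable on Niagara sprites, that procedurally draws hollow squares.
# It centers the UVs, raises the absolute value of each axis to a high power, then takes the maximum, producing a mask that brightens toward the square's edges without any texture.
# The boosted, saturated mask is multiplied by particle alpha.
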Material /BRCosmetics/Weapons/FORT_Melee/Pickaxe_TungStan/FX/M_Pickaxe_TungStan_Squares
graph {
 properties {
  bUsedWithNiagaraSprites=true
  blendMode=BLEND_Translucent
  pixelDepthOffsetMode=PDOM_Legacy
  shadingModel=MSM_Unlit
  twoSided=true
  usedShadingModels=Unlit
 }
 nodes [
  n1 = Subtract(A=TextureCoordinate()) {constB=0.5}
  n2 = Power(Base=Abs(None=ComponentMask(None=@n1) {g=true})) {constExponent=7}
  n3 = Power(Base=Abs(None=ComponentMask(None=@n1) {r=true})) {constExponent=7}
  n4 = Saturate(None=Max(A=Multiply(A=@n2) {constB=30}, B=Multiply(A=@n3) {constB=30}))
  Multiply(A=ParticleColor().A, B=Saturate(None=Multiply(A=@n4) {constB=4}))
 ]
}
