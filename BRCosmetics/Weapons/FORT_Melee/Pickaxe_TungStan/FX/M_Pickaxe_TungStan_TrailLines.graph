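# Unlit, translucent, two-sided material for the TungStan pickaxe's trail lines, set up for Niagara ribbons.
# A grayscale line texture, squared, blends particle color toward a bright HDR blue-violet, so the strongest parts of the lines take the fixed tint.
# The blend passes through the shared MF_FX_EmissiveAdjustments function for emissive, while particle alpha times the unsquared texture forms the opacity mask value.
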Material /BRCosmetics/Weapons/FORT_Melee/Pickaxe_TungStan/FX/M_Pickaxe_TungStan_TrailLines
graph {
 properties {
  bUsedWithNiagaraRibbons=true
  blendMode=BLEND_Translucent
  pixelDepthOffsetMode=PDOM_Legacy
  shadingModel=MSM_Unlit
  twoSided=true
  usedShadingModels=Unlit
 }
 nodes [
  n1 = ParticleColor()
  n2 = TextureSample() {samplerType=SAMPLERTYPE_LinearGrayscale, texture=&/T_PickAxe_TungStain_TrailLines}
  MaterialFunctionCall(Emissive=LinearInterpolate(A=@n1.RGB, B=Constant3Vector() {constant=<rgba 0.9251 0 5 1>}, Alpha=Multiply(A=@n2.R, B=@n2.R))) {materialFunction=&MF_FX_EmissiveAdjustments}
  Multiply(A=@n1.A, B=@n2.R)
 ]
}
