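# Additive, unlit, two-sided Niagara sprite material that draws a thin wavy ring for the TungStan pickaxe effects.
# A blurred noise texture is sampled in polar coordinates and scrolled radially over time, then offsets the sprite UVs.
# The ring is procedural: distance from center of the distorted UVs goes through Abs and an inverted-range SmoothStep, then is multiplied by particle alpha.
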Material /BRCosmetics/Weapons/FORT_Melee/Pickaxe_TungStan/FX/M_Pickaxe_TungStan_Waves
graph {
 properties {
  bUsedWithNiagaraSprites=true
  blendMode=BLEND_Additive
  pixelDepthOffsetMode=PDOM_Legacy
  shadingModel=MSM_Unlit
  twoSided=true
  usedShadingModels=Unlit
 }
 nodes [
  n1 = Subtract(A=TextureCoordinate()) {constB=0.5}
  n2 = ComponentMask(None=@n1) {r=true}
  n3 = ComponentMask(None=@n1) {g=true}
  n4 = Multiply(A=Multiply(A=Arctangent2Fast(Y=@n2, X=@n3), B=Divide() {constB=6.28})) {constB=2.2}
  n5 = Length(None=@n1) {is=class MaterialExpressionLength}
  n6 = Constant()
  n7 = Multiply(A=Time()) {constB=1.2}
  n8 = TextureSample(UVs=Add(A=AppendVector(A=@n4, B=@n5), B=AppendVector(A=@n6, B=@n7))) {texture=/Game/Effects/Fort_Effects/Emotes/Candor/T_Noise_Blur_Texture}
  n9 = Subtract(A=Add(A=TextureCoordinate(), B=Multiply(A=@n8.R) {constB=0.15})) {constB=0.5}
  n10 = Subtract(A=Multiply(A=Length(None=@n9)) {constB=1.25}) {constB=0.5}
  Multiply(A=ParticleColor().A, B=SmoothStep(Value=Abs(None=@n10)) {constMax=0.01, constMin=0.03})
 ]
}
