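# Niagara system for the TungStan pickaxe's purple animation trail.
# Layered ribbon emitters track bone/socket pairs on User.SkeletalMesh through FN_AnimTrail modules, narrowing and fading over short lifetimes with bright HDR color flashes on some layers.
# A sprite emitter adds purple sparks spawned by distance moved, pushed by curl noise and inherited movement.
# Ribbon alpha drops at night via a time-of-day selector; FN_AnimTrail005 is disabled.
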NiagaraSystem /BRCosmetics/Weapons/FORT_Melee/Pickaxe_TungStan/FX/NS_Pickaxe_TungStan_AnimTrail
dependencies {
 usedDataInterfaces [&NiagaraDataInterfaceColorCurve, &NiagaraDataInterfaceCurve]
 usedDynamicInputs []
 usedModules [&SystemState]
}
emitters {
 Empty {
  emitterUpdateScript [
   &EmitterState()
   &SpawnPerUnit(Module.Use Max Movement Threshold=true, Module.Use Movement Tolerance=false, Spawn Spacing=30, Use Spawn Probability=false)
  ]
  particleSpawnScript [&InitializeParticle(
    AdjustAlpha=false
    AdjustHue=true
    AdjustSaturation=false
    AdjustValue=false
    Color=<rgba 3.3067 0.0479 8 1>
    Color Mode=::NewEnumerator3
    Lifetime Max=0.3
    Lifetime Min=0.1
    Lifetime Mode=::NewEnumerator1
    Position Offset=<0 0 20>
    Sprite Size Mode=::NewEnumerator1
    Uniform Sprite Size Max=5
    Uniform Sprite Size Min=1.5
    UsePositionOffset=true
    Write Material Random=true
   )]
  particleUpdateScript [
   &ParticleState()
   &ScaleSpriteSize(Uniform Curve Sprite Scale=DataInterface() {propertyValues {Curve {keys [
        {interpMode=RCIM_Cubic, leaveTangent=4.6748, tangentMode=RCTM_Break, time=0, value=0.4}
        <at 0.2 1>
        {interpMode=RCIM_Cubic, leaveTangent=-0.4899, tangentMode=RCTM_Break, time=0.4, value=1}
        {arriveTangent=-1.4327, interpMode=RCIM_Cubic, tangentMode=RCTM_Break, time=1, value=0.35}
       ]}}})
   &CurlNoiseForce(Noise Frequency=20, Noise Strength=100, Pan Noise=false)
   &InheritSourceMovement(Applied Force Scale=<0.1 0.1 0.1>, ApplyForce=true, ApplyForceFalloff=false, ApplyPosition=false, Module.LimitForceSpeed=false)
   &SolveForcesAndVelocity(Clamp Velocity=false, Limit Acceleration=false)
   &ScaleColor(
    Linear Color Curve=DataInterface() {propertyValues {
      AlphaCurve {keys [<at 0 1>,<at 1 0.8>]}
      BlueCurve {keys [<at 0 1>,<at 0.6192 0.3385>]}
      GreenCurve {keys [<at 0 1>,<at 0.6192 0.0005>]}
      RedCurve {keys [<at 0 1>,<at 0.6192 0.0379>]}
     }}
    Scale Mode=::NewEnumerator2
   )
  ]
  rendererClasses [&NiagaraSpriteRendererProperties]
  renderers [{rendererClass=&NiagaraSpriteRendererProperties}]
 }
 FN_AnimTrail {
  emitterSpawnScript [&FN_RibbonIDOffsetUtility()]
  emitterUpdateScript [
   &EmitterState()
   &FN_SpawnAnimTrail(Number of Anim Trails=2)
  ]
  particleSpawnScript [
   &FN_InitializeAnimTrailBoneIndecies(First Bone or Socket Index=2, Second Bone or Socket Index=3)
   &FN_InitializeAnimTrailBoneIndecies(AnimTrail Selection=1, First Bone or Socket Index=4, Second Bone or Socket Index=5) {module=FN_InitializeAnimTrailBoneIndecies001}
   &FN_InitializeAnimTrail(Color=<rgba 0.0287 0 0.1615 1>, Lifetime=0.12)
  ]
  particleUpdateScript [
   &FN_UpdateAnimTrail(Locked Particle RibbonWidth:NiagaraFloat=$Particles.RibbonWidth, SkeletalMesh=$User.SkeletalMesh) {module=FN_UpdateAnimTrail001}
   &Color(Color=&MakeLinearColor(
     R:NiagaraFloat=1
     G:NiagaraFloat=1
     B:NiagaraFloat=1
     A:NiagaraFloat=&FN_Select_Float_from_Time_Of_Day(Morning=1, Day=1, Evening=1, Night=0.3)
    ))
   &ScaleColor(
    Linear Color Curve=DataInterface() {propertyValues {
      AlphaCurve {keys [<at 0 0.3>,<at 0.2358 1>,<at 0.6828 1>,<at 1 0>]}
      BlueCurve {keys [<at 0 1>,<at 0.1134 1>,<at 1 0.5573>]}
      GreenCurve {keys [<at 0 1>,<at 0.1134 1>,<at 1 0.1036>]}
      RedCurve {keys [<at 0 1>,<at 0.1134 1>,<at 1 0.3027>]}
     }}
    Scale Mode=::NewEnumerator2
   )
   &ScaleRibbonWidth(Ribbon Width Scale=&FloatFromCurve(FloatCurve=DataInterface() {propertyValues {Curve {keys [
         {interpMode=RCIM_Cubic, tangentMode=RCTM_User, time=0, value=1}
         {arriveTangent=-1.9514, interpMode=RCIM_Cubic, leaveTangent=-1.9514, tangentMode=RCTM_User, time=1, value=0}
        ]}}}))
   &SolveForcesAndVelocity(Clamp Velocity=false, Limit Acceleration=false)
  ]
  rendererClasses [&NiagaraRibbonRendererProperties]
  renderers [{rendererClass=&NiagaraRibbonRendererProperties}]
 }
 FN_AnimTrail001 {
  emitterSpawnScript [&FN_RibbonIDOffsetUtility()]
  emitterUpdateScript [
   &EmitterState()
   &FN_SpawnAnimTrail()
  ]
  particleSpawnScript [
   &FN_InitializeAnimTrailBoneIndecies()
   &FN_InitializeAnimTrail(AnimTrailWidth=1.2, Color=<rgba 0.1405 0.028 0.5833 0.6>, Lifetime=0.11)
  ]
  particleUpdateScript [
   &FN_UpdateAnimTrail(Locked Particle RibbonWidth:NiagaraFloat=$Particles.RibbonWidth, SkeletalMesh=$User.SkeletalMesh) {module=FN_UpdateAnimTrail001}
   &Color(Color=&MakeLinearColor(
     R:NiagaraFloat=1
     G:NiagaraFloat=1
     B:NiagaraFloat=1
     A:NiagaraFloat=&FN_Select_Float_from_Time_Of_Day(Morning=1, Day=1, Evening=1, Night=0.3)
    ))
   &ScaleColor(
    Linear Color Curve=DataInterface() {propertyValues {
      AlphaCurve {keys [<at 0 1>,<at 1 0>]}
      BlueCurve {keys [<at 0 1>,<at 0.4938 0.1771>]}
      GreenCurve {keys [<at 0 1>,<at 0.4938 0.0029>]}
      RedCurve {keys [<at 0 1>,<at 0.4938 0.0192>]}
     }}
    Scale Mode=::NewEnumerator2
   )
   &ScaleRibbonWidth(Ribbon Width Scale=&FloatFromCurve(FloatCurve=DataInterface() {propertyValues {Curve {keys [
         {interpMode=RCIM_Cubic, tangentMode=RCTM_User, time=0, value=1}
         {arriveTangent=-1.9514, interpMode=RCIM_Cubic, leaveTangent=-1.9514, tangentMode=RCTM_User, time=1, value=0}
        ]}}}))
   &SolveForcesAndVelocity(Clamp Velocity=false, Limit Acceleration=false)
  ]
  rendererClasses [&NiagaraRibbonRendererProperties]
  renderers [{rendererClass=&NiagaraRibbonRendererProperties}]
 }
 FN_AnimTrail002 {
  emitterSpawnScript [&FN_RibbonIDOffsetUtility()]
  emitterUpdateScript [
   &EmitterState()
   &FN_SpawnAnimTrail()
  ]
  particleSpawnScript [
   &FN_InitializeAnimTrailBoneIndecies(First Bone or Socket Index=6, Second Bone or Socket Index=7)
   &FN_InitializeAnimTrail(AnimTrailWidth=1.2, Color=<rgba 0.1405 0.028 0.5833 0.45>, Lifetime=0.1)
  ]
  particleUpdateScript [
   &FN_UpdateAnimTrail(Locked Particle RibbonWidth:NiagaraFloat=$Particles.RibbonWidth, SkeletalMesh=$User.SkeletalMesh) {module=FN_UpdateAnimTrail001}
   &Color(Color=&MakeLinearColor(
     R:NiagaraFloat=1
     G:NiagaraFloat=1
     B:NiagaraFloat=1
     A:NiagaraFloat=&FN_Select_Float_from_Time_Of_Day(Morning=1, Day=1, Evening=1, Night=0.3)
    ))
   &ScaleColor(
    Linear Color Curve=DataInterface() {propertyValues {
      AlphaCurve {keys [<at 0 1>,<at 1 0>]}
      BlueCurve {keys [<at 0 1>,<at 0.4938 0.1771>]}
      GreenCurve {keys [<at 0 1>,<at 0.4938 0.0029>]}
      RedCurve {keys [<at 0 1>,<at 0.4938 0.0192>]}
     }}
    Scale Mode=::NewEnumerator2
   )
   &ScaleRibbonWidth(Ribbon Width Scale=&FloatFromCurve(FloatCurve=DataInterface() {propertyValues {Curve {keys [
         {interpMode=RCIM_Cubic, tangentMode=RCTM_User, time=0, value=1}
         {arriveTangent=-1.9514, interpMode=RCIM_Cubic, leaveTangent=-1.9514, tangentMode=RCTM_User, time=1, value=0}
        ]}}}))
   &SolveForcesAndVelocity(Clamp Velocity=false, Limit Acceleration=false)
  ]
  rendererClasses [&NiagaraRibbonRendererProperties]
  renderers [{rendererClass=&NiagaraRibbonRendererProperties}]
 }
 FN_AnimTrail003 {
  emitterSpawnScript [&FN_RibbonIDOffsetUtility()]
  emitterUpdateScript [
   &EmitterState()
   &FN_SpawnAnimTrail()
  ]
  particleSpawnScript [
   &FN_InitializeAnimTrailBoneIndecies()
   &FN_InitializeAnimTrail(Color=<rgba 3.6137 0.7196 15 0.6>, Lifetime=0.12)
  ]
  particleUpdateScript [
   &FN_UpdateAnimTrail(Locked Particle RibbonWidth:NiagaraFloat=$Particles.RibbonWidth, SkeletalMesh=$User.SkeletalMesh) {module=FN_UpdateAnimTrail001}
   &Color(Color=&MakeLinearColor(
     R:NiagaraFloat=1
     G:NiagaraFloat=1
     B:NiagaraFloat=1
     A:NiagaraFloat=&FN_Select_Float_from_Time_Of_Day(Morning=1, Day=1, Evening=1, Night=0.3)
    ))
   &ScaleColor(
    Linear Color Curve=DataInterface() {propertyValues {
      AlphaCurve {keys [<at 0 1>,<at 1 0>]}
      BlueCurve {keys [<at 0 5>,<at 0.0445 5>,<at 0.0667 0.2874>,<at 0.1246 0.2874>,<at 0.149 5>,<at 0.2447 5>,<at 0.2669 0.2874>,<at 0.3203 0.2874>,<at 0.3536 5>,<at 0.7451 0.1771>]}
      GreenCurve {keys [<at 0 3.2101>,<at 0.0445 4.0625>,<at 0.0667 0.0232>,<at 0.1246 0.0232>,<at 0.149 2.6936>,<at 0.2447 2.6936>,<at 0.2669 0.0232>,<at 0.3203 0.0232>,<at 0.3536 4.6615>,<at 0.7451 0.0029>]}
      RedCurve {keys [<at 0 4.3546>,<at 0.0445 4.3267>,<at 0.0667 0.0976>,<at 0.1246 0.0976>,<at 0.149 4.096>,<at 0.2447 4.096>,<at 0.2669 0.0976>,<at 0.3203 0.0976>,<at 0.3536 4.7568>,<at 0.7451 0.0192>]}
     }}
    Scale Mode=::NewEnumerator2
   ) {enabled=false}
   &ScaleColor(
    Linear Color Curve=DataInterface() {propertyValues {
      AlphaCurve {keys [<at 0 1>,<at 1 0>]}
      BlueCurve {keys [<at 0.0044 1>,<at 0.089 50>,<at 0.4003 0.1771>]}
      GreenCurve {keys [<at 0.0044 0.642>,<at 0.089 32.1012>,<at 0.4003 0.0029>]}
      RedCurve {keys [<at 0.0044 0.8709>,<at 0.089 43.5463>,<at 0.4003 0.0192>]}
     }}
    Scale Mode=::NewEnumerator2
   ) {module=ScaleColor001}
   &ScaleRibbonWidth(Ribbon Width Scale=&FloatFromCurve(FloatCurve=DataInterface() {propertyValues {Curve {keys [
         {interpMode=RCIM_Cubic, tangentMode=RCTM_User, time=0, value=1}
         {arriveTangent=-1.9514, interpMode=RCIM_Cubic, leaveTangent=-1.9514, tangentMode=RCTM_User, time=1, value=0}
        ]}}}))
   &SolveForcesAndVelocity(Clamp Velocity=false, Limit Acceleration=false)
  ]
  rendererClasses [&NiagaraRibbonRendererProperties]
  renderers [{rendererClass=&NiagaraRibbonRendererProperties}]
 }
 FN_AnimTrail004 {
  emitterSpawnScript [&FN_RibbonIDOffsetUtility()]
  emitterUpdateScript [
   &EmitterState()
   &FN_SpawnAnimTrail()
  ]
  particleSpawnScript [
   &FN_InitializeAnimTrailBoneIndecies(First Bone or Socket Index=6, Second Bone or Socket Index=7)
   &FN_InitializeAnimTrail(AnimTrailWidth=0.9, Color=<rgba 3.6137 0.7196 15 0.4>, Lifetime=0.1)
  ]
  particleUpdateScript [
   &FN_UpdateAnimTrail(Locked Particle RibbonWidth:NiagaraFloat=$Particles.RibbonWidth, SkeletalMesh=$User.SkeletalMesh) {module=FN_UpdateAnimTrail001}
   &Color(Color=&MakeLinearColor(
     R:NiagaraFloat=1
     G:NiagaraFloat=1
     B:NiagaraFloat=1
     A:NiagaraFloat=&FN_Select_Float_from_Time_Of_Day(Morning=1, Day=1, Evening=1, Night=0.3)
    ))
   &ScaleColor(
    Linear Color Curve=DataInterface() {propertyValues {
      AlphaCurve {keys [<at 0 1>,<at 1 0>]}
      BlueCurve {keys [<at 0 2>,<at 0.089 40>,<at 0.3648 0.1771>]}
      GreenCurve {keys [<at 0 1.284>,<at 0.089 2.8792>,<at 0.3648 0.0029>]}
      RedCurve {keys [<at 0 1.7419>,<at 0.089 23.0317>,<at 0.3648 0.0192>]}
     }}
    Scale Mode=::NewEnumerator2
   )
   &ScaleRibbonWidth(Ribbon Width Scale=&FloatFromCurve(FloatCurve=DataInterface() {propertyValues {Curve {keys [
         {interpMode=RCIM_Cubic, tangentMode=RCTM_User, time=0, value=1}
         {arriveTangent=-1.9514, interpMode=RCIM_Cubic, leaveTangent=-1.9514, tangentMode=RCTM_User, time=1, value=0}
        ]}}}))
   &SolveForcesAndVelocity(Clamp Velocity=false, Limit Acceleration=false)
  ]
  rendererClasses [&NiagaraRibbonRendererProperties]
  renderers [{rendererClass=&NiagaraRibbonRendererProperties}]
 }
 FN_AnimTrail005 {
  bEnabled=false
  emitterSpawnScript [&FN_RibbonIDOffsetUtility()]
  emitterUpdateScript [
   &EmitterState()
   &FN_SpawnAnimTrail(Number of Anim Trails=2)
  ]
  particleSpawnScript [
   &FN_InitializeAnimTrailBoneIndecies(First Bone or Socket Index=2, Second Bone or Socket Index=3)
   &FN_InitializeAnimTrailBoneIndecies(First Bone or Socket Index=4, Second Bone or Socket Index=5) {module=FN_InitializeAnimTrailBoneIndecies001}
   &FN_InitializeAnimTrail(AnimTrailWidth=0.8, Color=<rgba 3.6137 0.7196 15 0.6>, Lifetime=0.12)
  ]
  particleUpdateScript [
   &FN_UpdateAnimTrail(Locked Particle RibbonWidth:NiagaraFloat=$Particles.RibbonWidth, SkeletalMesh=$User.SkeletalMesh) {module=FN_UpdateAnimTrail001}
   &Color(Color=&MakeLinearColor(
     R:NiagaraFloat=1
     G:NiagaraFloat=1
     B:NiagaraFloat=1
     A:NiagaraFloat=&FN_Select_Float_from_Time_Of_Day(Morning=1, Day=1, Evening=1, Night=0.3)
    ))
   &ScaleColor(
    Linear Color Curve=DataInterface() {propertyValues {
      AlphaCurve {keys [<at 0 1>,<at 1 0>]}
      BlueCurve {keys [<at 0 5>,<at 0.0445 5>,<at 0.0667 0.2874>,<at 0.1246 0.2874>,<at 0.149 5>,<at 0.2447 5>,<at 0.2669 0.2874>,<at 0.3203 0.2874>,<at 0.3536 5>,<at 0.7451 0.1771>]}
      GreenCurve {keys [<at 0 3.2101>,<at 0.0445 4.0625>,<at 0.0667 0.0232>,<at 0.1246 0.0232>,<at 0.149 2.6936>,<at 0.2447 2.6936>,<at 0.2669 0.0232>,<at 0.3203 0.0232>,<at 0.3536 4.6615>,<at 0.7451 0.0029>]}
      RedCurve {keys [<at 0 4.3546>,<at 0.0445 4.3267>,<at 0.0667 0.0976>,<at 0.1246 0.0976>,<at 0.149 4.096>,<at 0.2447 4.096>,<at 0.2669 0.0976>,<at 0.3203 0.0976>,<at 0.3536 4.7568>,<at 0.7451 0.0192>]}
     }}
    Scale Mode=::NewEnumerator2
   ) {enabled=false}
   &ScaleColor(
    Linear Color Curve=DataInterface() {propertyValues {
      AlphaCurve {keys [<at 0 1>,<at 1 0>]}
      BlueCurve {keys [<at 0.0044 1>,<at 0.089 50>,<at 0.4003 0.1771>]}
      GreenCurve {keys [<at 0.0044 0.642>,<at 0.089 32.1012>,<at 0.4003 0.0029>]}
      RedCurve {keys [<at 0.0044 0.8709>,<at 0.089 43.5463>,<at 0.4003 0.0192>]}
     }}
    Scale Mode=::NewEnumerator2
   ) {module=ScaleColor001}
   &ScaleRibbonWidth(Ribbon Width Scale=&FloatFromCurve(FloatCurve=DataInterface() {propertyValues {Curve {keys [
         {interpMode=RCIM_Cubic, tangentMode=RCTM_User, time=0, value=1}
         {arriveTangent=-1.9514, interpMode=RCIM_Cubic, leaveTangent=-1.9514, tangentMode=RCTM_User, time=1, value=0}
        ]}}}))
   &SolveForcesAndVelocity(Clamp Velocity=false, Limit Acceleration=false)
  ]
  rendererClasses [&NiagaraRibbonRendererProperties]
  renderers [{rendererClass=&NiagaraRibbonRendererProperties}]
 }
}
summary {
 systemName=NS_Pickaxe_TungStan_AnimTrail
 emitters [
  {emitterName=FN_AnimTrail001, rendererClasses [&NiagaraRibbonRendererProperties]}
  {emitterName=FN_AnimTrail, rendererClasses [&NiagaraRibbonRendererProperties]}
  {emitterName=FN_AnimTrail002, rendererClasses [&NiagaraRibbonRendererProperties]}
  {emitterName=Empty, rendererClasses [&NiagaraSpriteRendererProperties]}
  {emitterName=FN_AnimTrail003, rendererClasses [&NiagaraRibbonRendererProperties]}
  {emitterName=FN_AnimTrail004, rendererClasses [&NiagaraRibbonRendererProperties]}
  {bEnabled=false, emitterName=FN_AnimTrail005, rendererClasses [&NiagaraRibbonRendererProperties]}
 ]
 userVariables [
  User.RibbonIDOffset {type=NiagaraInt32}
  User.SkeletalMesh {
   type=NiagaraDataInterfaceSkeletalMesh
   defaultValue {
    struct=NiagaraExt_VariableValue_DataInterface
    value {dataInterface=&self:NiagaraDataInterfaceSkeletalMesh_0}
   }
  }
 ]
}
