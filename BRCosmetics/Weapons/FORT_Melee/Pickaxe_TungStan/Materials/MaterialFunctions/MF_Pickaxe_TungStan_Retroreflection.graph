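# Material function for the TungStan pickaxe that builds a retroreflective honeycomb and stripe finish, outputting base color, emissive, roughness and clear coat.
# A unique mask texture limits the finish and clear coat and controls handle darkening, while tiled masks supply honeycomb and stripe patterns.
# Emissive fakes retroreflection with a powered inverse Fresnel through the honeycomb, tinted by base color and scaled by time of day.
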MaterialFunction /BRCosmetics/Weapons/FORT_Melee/Pickaxe_TungStan/Materials/MaterialFunctions/MF_Pickaxe_TungStan_Retroreflection
graph {nodes [
  n1 = FunctionInput() {bUsePreviewValueAsDefault=true, inputName=BaseColorMultiplier, inputType=FunctionInput_Scalar, previewValue=<1 0 0 1>}
  n2 = Multiply(A=FunctionInput() {inputName=BaseColor}, B=@n1)
  n3 = FunctionInput() {bUsePreviewValueAsDefault=true, inputName=StripesScale, inputType=FunctionInput_Scalar, sortPriority=5, previewValue=<1 0 0 1>}
  n4 = TextureSample(UVs=Multiply(A=TextureCoordinate(), B=Multiply(A=@n3) {constB=10})) {samplerType=SAMPLERTYPE_Masks, texture=/BRCosmetics/Weapons/FORT_Melee/Pickaxe_TungStan/Textures/T_Pickaxe_Tungstan_TileableMasks}
  n5 = FunctionInput() {bUsePreviewValueAsDefault=true, inputName=StripesOpacity, inputType=FunctionInput_Scalar, sortPriority=5, previewValue=<0.1 0 0 1>}
  n6 = Add(A=@n2, B=Saturate(None=Multiply(A=@n4.B, B=@n5) {constB=0.1}))
  n7 = TextureSample() {samplerType=SAMPLERTYPE_LinearColor, texture=/BRCosmetics/Weapons/FORT_Melee/Pickaxe_TungStan/Textures/T_Pickaxe_Tungstan_Masks}
  n8 = FunctionInput() {bUsePreviewValueAsDefault=true, inputName=HoneycombTextureSize, inputType=FunctionInput_Scalar, sortPriority=4, previewValue=<1 0 0 1>}
  n9 = TextureSample(UVs=Multiply(A=TextureCoordinate(), B=Multiply(A=@n8) {constB=2}) {constB=10}) {samplerType=SAMPLERTYPE_Masks, texture=&T_Pickaxe_Tungstan_TileableMasks}
  n10 = Add(
   A=@n9.R
   B=OneMinus(None=FunctionInput() {bUsePreviewValueAsDefault=true, inputName=HoneycombColorOpacity, inputType=FunctionInput_Scalar, sortPriority=4})
  ) {constB=0.2}
  n11 = LinearInterpolate(A=@n2, B=Multiply(A=@n2, B=Saturate(None=@n10)), Alpha=@n7.R)
  n12 = FunctionInput() {bUsePreviewValueAsDefault=true, inputName=BlackHandleMultiplier, inputType=FunctionInput_Scalar, sortPriority=1, previewValue=<0.2 0 0 1>}
  n13 = Multiply(A=@n11, B=@n12)
  FunctionOutput(None=Add(A=LinearInterpolate(A=@n2, B=@n6, Alpha=@n7.R), B=LinearInterpolate(A=@n13, B=@n11, Alpha=@n7.G) {constAlpha=0.9})) {outputName=BaseColor}
  n14 = OneMinus(None=Fresnel() {exponent=1})
  n15 = FunctionInput() {bUsePreviewValueAsDefault=true, inputName=RetroReflectiveFresnelExp, inputType=FunctionInput_Scalar, sortPriority=4, previewValue=<20 0 0 1>}
  n16 = Multiply(A=Multiply(A=Power(Base=@n14, Exp=@n15) {constExponent=14}, B=@n9.R), B=@n7.R)
  n17 = FunctionInput() {
   bUsePreviewValueAsDefault=true
   inputName=RetroreflectiveIntensityDay
   inputType=FunctionInput_Scalar
   sortPriority=7
   previewValue=<3 0 0 1>
  }
  n18 = FunctionInput() {
   bUsePreviewValueAsDefault=true
   inputName=RetroreflectiveIntensityEvening
   inputType=FunctionInput_Scalar
   sortPriority=7
   previewValue=<2 0 0 1>
  }
  n19 = FunctionInput() {
   bUsePreviewValueAsDefault=true
   inputName=RetroreflectiveIntensityNight
   inputType=FunctionInput_Scalar
   sortPriority=7
   previewValue=<1 0 0 1>
  }
  FunctionOutput(None=Multiply(
    A=Multiply(A=@n2, B=@n16)
    B=MaterialFunctionCall(Value_Morning=@n17, Value_Day=@n17, Value_Evening=@n18, Value_Night=@n19) {materialFunction=&Scalar_Time_of_Day}.Result
   ) {constB=2}) {outputName=EmissiveColor, sortPriority=1}
  n20 = FunctionInput() {bUsePreviewValueAsDefault=true, inputName=Roughness, inputType=FunctionInput_Scalar, sortPriority=2, previewValue=<0.5 0 0 1>}
  n21 = FunctionInput() {bUsePreviewValueAsDefault=true, inputName=RoughnessRemap, inputType=FunctionInput_Scalar, sortPriority=2}
  n22 = FunctionInput() {
   bUsePreviewValueAsDefault=true
   inputName=StripesRoughnessMultiplier
   inputType=FunctionInput_Scalar
   sortPriority=5
   previewValue=<0.1 0 0 1>
  }
  n23 = Add(A=@n4.B, B=@n22)
  n24 = FunctionInput() {bUsePreviewValueAsDefault=true, inputName=HoneycombRoughnessOpacity, inputType=FunctionInput_Scalar, sortPriority=4}
  n25 = Multiply(A=Multiply(A=LinearInterpolate(A=@n21, Alpha=@n20), B=Saturate(None=@n23)), B=Saturate(None=Add(A=@n9.R, B=@n24) {constB=0.2}))
  n26 = FunctionInput() {bUsePreviewValueAsDefault=true, inputName=HegaxonRoughnessVariation, inputType=FunctionInput_Scalar, previewValue=<1 0 0 1>}
  n27 = Add(A=@n26) {constB=-1}
  n28 = OneMinus(None=Saturate(None=Add(A=@n9.G, B=@n27) {constB=-1}))
  FunctionOutput(None=LinearInterpolate(A=@n20, B=Multiply(A=@n25, B=@n28), Alpha=@n7.R)) {outputName=Roughness, sortPriority=2}
  n29 = FunctionInput() {bUsePreviewValueAsDefault=true, inputName=ClearCoatRoughness, inputType=FunctionInput_Scalar, sortPriority=3, previewValue=<0.5 0 0 1>}
  FunctionOutput(None=@n29) {outputName=ClearCoatRoughness, sortPriority=4}
  FunctionOutput(None=@n7.R) {outputName=ClearCoat, sortPriority=3}
 ]}
