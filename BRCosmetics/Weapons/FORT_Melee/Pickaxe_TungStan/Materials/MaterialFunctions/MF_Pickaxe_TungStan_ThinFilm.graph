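# Material function for the TungStan pickaxe that adds a thin-film color sheen to an input BaseColor.
# It samples a shared character thin-film texture with rotated, scaled UVs, hue-shifts it, scales it by ThinFilmIntensity and adds it to the base color.
# A sharpened Fresnel term and the red channel of the pickaxe mask texture limit where the tinted result replaces the original color.
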MaterialFunction /BRCosmetics/Weapons/FORT_Melee/Pickaxe_TungStan/Materials/MaterialFunctions/MF_Pickaxe_TungStan_ThinFilm
graph {nodes [
  n1 = FunctionInput() {inputName=BaseColor}
  n2 = Constant() {r=0.18}
  n3 = TextureCoordinate()
  n4 = Constant() {r=0.25}
  n5 = TextureSample(UVs=Multiply(A=MaterialFunctionCall(UVs=@n3, "Rotation Angle (0-1)"=@n4) {materialFunction=&CustomRotator}.Rotated Values) {constB=1.2}) {
   samplerType=SAMPLERTYPE_Masks
   texture=/Game/Characters/Player/Male/Medium/Bodies/M_MED_SunBurst_Alt/Skins/AltStyle/Textures/T_F_MED_Global_FB_ThinFilm_color
  }
  n6 = FunctionInput() {bUsePreviewValueAsDefault=true, inputName=ThinFilmIntensity, inputType=FunctionInput_Scalar, sortPriority=1, previewValue=<0.5 0 0 1>}
  n7 = Add(
   A=@n1
   B=Multiply(
    A=MaterialFunctionCall(Hue Shift Percentage=@n2, Texture=@n5.RGB) {materialFunction=&HueShift}.Result
    B=@n6
   ) {constB=0.56}
  )
  n8 = Power(Base=Fresnel() {baseReflectFraction=0.1, exponent=0.1}) {constExponent=32}
  FunctionOutput(None=LinearInterpolate(
    A=@n1
    B=LinearInterpolate(A=@n1, B=@n7, Alpha=@n8)
    Alpha=TextureSample() {samplerType=SAMPLERTYPE_LinearColor, texture=/BRCosmetics/Weapons/FORT_Melee/Pickaxe_TungStan/Textures/T_Pickaxe_Tungstan_Masks}.R
   )) {outputName=BaseColor}
 ]}
