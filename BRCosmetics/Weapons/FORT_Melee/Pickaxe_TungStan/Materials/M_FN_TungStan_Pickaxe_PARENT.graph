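# Parent material for the TungStan pickaxe cosmetic, adding a retroreflective, thin-film finish to the shared character material chain.
# Base PreFX, CharacterSkinFX, PostFX and cinematic PostFX functions produce the attributes, whose base color and roughness feed a pickaxe retroreflection function.
# That function exposes honeycomb, stripe, clear coat and day, evening and night intensity parameters.
# Its base color passes through a thin-film function before the results override the material attributes.
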Material /BRCosmetics/Weapons/FORT_Melee/Pickaxe_TungStan/Materials/M_FN_TungStan_Pickaxe_PARENT
graph {
 properties {
  bAutomaticallySetUsageInEditor=false
  bUseMaterialAttributes=true
  bUsedWithClothing=true
  bUsedWithMeshDeformer=true
  bUsedWithMorphTargets=true
  bUsedWithNanite=true
  bUsedWithSkeletalMesh=true
  pixelDepthOffsetMode=PDOM_Legacy
  refractionMethod=RM_IndexOfRefraction
  shadingModel=MSM_FromMaterialExpression
  subsurfaceProfile=&SS_HeroSkin_02
  usedShadingModels=Default Lit | Subsurface | Clear Coat | Subsurface Profile | Cloth | Eye
 }
 nodes [
  ScalarParameter() {defaultValue=1, group=RetroreflectiveParams, parameterName=HoneycombNormalIntensity}
  n1 = MaterialFunctionCall() {materialFunction=&BaseFN_CharacterFunction_PreFX}
  n2 = MaterialFunctionCall(
   AO=@n1.AO
   Distance Blend=@n1.AO
   Fresnel=@n1.AO
   Gmap=@n1.AO
   Pre Skinned Local Position=@n1.AO
   Material=MaterialFunctionCall(
    PreskinnedPosition=@n1.Pre skinned Local Position
    Gmap=@n1.Pre skinned Local Position
    MaterialAttributes=@n1.Pre skinned Local Position
    AO=@n1.Pre skinned Local Position
   ) {materialFunction=&CharacterSkinFX}.Result
  ) {materialFunction=&BaseFN_CharacterFunction_PostFX}
  n3 = GetMaterialAttributes(None=MaterialFunctionCall(In=@n2.Result) {materialFunction=&CINE_MF_CharPostFX}.Result) {attributeGetTypes [#1, #2, #3]}
  n4 = MaterialFunctionCall!(
   HegaxonRoughnessVariation=ScalarParameter() {group=RetroreflectiveParams, parameterName=HoneycombRoughnessVariation}
   BaseColor=@n3.Base Color
   BaseColorMultiplier=ScalarParameter() {defaultValue=1, group=RetroreflectiveParams, parameterName=DiffuseColorMultiplier}
   BlackHandleMultiplier=ScalarParameter() {defaultValue=0.2, group=RetroreflectiveParams, parameterName=BlackHandleMultiplier}
   RoughnessRemap=ScalarParameter() {group=RetroreflectiveParams, parameterName=RoughnessRemapValue}
   Roughness=@n3.Roughness
   ClearCoatRoughness=ScalarParameter() {defaultValue=0.1, group=RetroreflectiveParams, parameterName=ClearCoatRoughness}
   HoneycombTextureSize=ScalarParameter() {defaultValue=10, group=RetroreflectiveParams, parameterName=HoneycombTextureSize}
   HoneycombColorOpacity=ScalarParameter() {group=RetroreflectiveParams, parameterName=HoneyCombColorOpacity}
   RetroReflectiveFresnelExp=ScalarParameter() {defaultValue=20, group=RetroreflectiveParams, parameterName=RetroreflectiveFresnelExp}
   HoneycombRoughnessOpacity=ScalarParameter() {defaultValue=1, group=RetroreflectiveParams, parameterName=HoneyCombRoughnessMultiplier}
   StripesScale=ScalarParameter() {defaultValue=1, group=RetroreflectiveParams, parameterName=StripesScale}
   StripesOpacity=ScalarParameter() {group=RetroreflectiveParams, parameterName=StripesOpacity}
   StripesRoughnessMultiplier=ScalarParameter() {group=RetroreflectiveParams, parameterName=StripesRoughnessMultiplier}
   RetroreflectiveIntensityEvening=ScalarParameter() {defaultValue=2, group=RetroreflectiveParams, parameterName=RetroreflectiveIntensityEvening}
   RetroreflectiveIntensityDay=ScalarParameter() {defaultValue=3, group=RetroreflectiveParams, parameterName=RetroreflectiveIntensityDay}
   RetroreflectiveIntensityNight=ScalarParameter() {defaultValue=1, group=RetroreflectiveParams, parameterName=RetroreflectiveIntensityNIght}
  ) {materialFunction=/BRCosmetics/Weapons/FORT_Melee/Pickaxe_TungStan/Materials/MaterialFunctions/MF_Pickaxe_TungStan_Retroreflection}
  SetMaterialAttributes!(
   MaterialAttributes=@n3.MaterialAttributes
   Base Color=MaterialFunctionCall!(
    BaseColor=@n4.BaseColor
    ThinFilmIntensity=ScalarParameter() {defaultValue=0.5, group=RetroreflectiveParams, parameterName=ThinFilmIntensity}
   ) {materialFunction=/BRCosmetics/Weapons/FORT_Melee/Pickaxe_TungStan/Materials/MaterialFunctions/MF_Pickaxe_TungStan_ThinFilm}.BaseColor
   Emissive Color=@n4.EmissiveColor
   Roughness=@n4.EmissiveColor
   Clear Coat or Cloth or Iris Mask or Curvature or Custom Data 0=@n4.EmissiveColor
   Clear Coat Roughness or Iris Distance or Custom Data 1=@n4.EmissiveColor
   Normal
  ) {attributeSetTypes [
    #1
    #4
    #2
    #5
    #6
    #3
   ]}
 ]
}
