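# Physics asset for the TungStan two-handed pickaxe, giving its six side panels bodies that can rotate about a hinge.
# Tiny axis and panel anchor bodies chain back to the handle capsule through constraints with free swing and twist, following upper and lower left/right branches.
# Each larger dyn_ panel capsule attaches to its matching anchor with both swings locked, leaving twist as the only rotation axis.
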PhysicsAsset /BRCosmetics/Weapons/FORT_Melee/Pickaxe_TungStan/Meshes/TungStan_Axe_Physics
bodies {
 Melee_TwoHanded_Handle [{shapeType=Capsule, length=31.9504, radius=2.6782, center=<0.0065 -0.054 10.3061>, rotation=<rot 0.1022 -90 0.0165>}]
 dyn_lower_panel_L [{shapeType=Capsule, length=8.5, radius=2, shapeName=lower_panel_L_capsule, center=<2.3 0 -1.8468>}]
 dyn_lower_panel_R [{shapeType=Capsule, length=8.5, radius=2, shapeName=lower_panel_R_capsule, center=<-2.3 0 -1.8468>}]
 dyn_upper_panel_01_L [{shapeType=Capsule, length=8.5, radius=2, shapeName=upper_panel_01_L_capsule, center=<2.3 0 -2.9276>}]
 dyn_upper_panel_01_R [{shapeType=Capsule, length=8.5, radius=2, shapeName=upper_panel_01_R_capsule, center=<-2.3 0 -2.9276>}]
 dyn_upper_panel_02_L [{shapeType=Capsule, length=8.5, radius=2, shapeName=upper_panel_02_L_capsule, center=<2.3 0 -2.9276>}]
 dyn_upper_panel_02_R [{shapeType=Capsule, length=8.5, radius=2, shapeName=upper_panel_02_R_capsule, center=<-2.3 0 -2.9276>}]
 lower_axis [{shapeType=Capsule, length=0, radius=0.5, shapeName=lower_axis_capsule}]
 lower_panel_L [{shapeType=Capsule, length=0, radius=0.5, shapeName=lower_panel_L_capsule, center=<2.3 0 -1.8468>}]
 lower_panel_R [{shapeType=Capsule, length=0, radius=0.5, shapeName=lower_panel_R_capsule, center=<-2.3 0 -1.8468>}]
 main_upper_axis [{shapeType=Capsule, length=0, radius=0.5, shapeName=main_upper_axis_capsule}]
 upper_axis_L [{shapeType=Capsule, length=0, radius=0.5, shapeName=upper_axis_L_capsule}]
 upper_axis_R [{shapeType=Capsule, length=0, radius=0.5, shapeName=upper_axis_R_capsule}]
 upper_panel_01_L [{shapeType=Capsule, length=0, radius=0.5, shapeName=upper_panel_01_L_capsule, center=<2.3 0 -2.9276>}]
 upper_panel_01_R [{shapeType=Capsule, length=0, radius=0.5, shapeName=upper_panel_01_R_capsule, center=<-2.3 0 -2.9276>}]
 upper_panel_02_L [{shapeType=Capsule, length=0, radius=0.5, shapeName=upper_panel_02_L_capsule, center=<2.3 0 -2.9276>}]
 upper_panel_02_R [{shapeType=Capsule, length=0, radius=0.5, shapeName=upper_panel_02_R_capsule, center=<-2.3 0 -2.9276>}]
}
constraints [
 {
  bone1Name=lower_axis
  bone2Name=Melee_TwoHanded_Handle
  swing1LimitDegrees=45
  swing1Motion=Free
  swing2LimitDegrees=45
  swing2Motion=Free
  twistLimitDegrees=45
  twistMotion=Free
 }
 {
  bone1Name=main_upper_axis
  bone2Name=Melee_TwoHanded_Handle
  swing1LimitDegrees=45
  swing1Motion=Free
  swing2LimitDegrees=45
  swing2Motion=Free
  twistLimitDegrees=45
  twistMotion=Free
 }
 {
  bone1Name=upper_axis_L
  bone2Name=main_upper_axis
  swing1LimitDegrees=45
  swing1Motion=Free
  swing2LimitDegrees=45
  swing2Motion=Free
  twistLimitDegrees=45
  twistMotion=Free
 }
 {
  bone1Name=upper_axis_R
  bone2Name=main_upper_axis
  swing1LimitDegrees=45
  swing1Motion=Free
  swing2LimitDegrees=45
  swing2Motion=Free
  twistLimitDegrees=45
  twistMotion=Free
 }
 {
  bone1Name=dyn_lower_panel_L
  bone2Name=lower_panel_L
  swing1LimitDegrees=45
  swing1Motion=Locked
  swing2LimitDegrees=45
  swing2Motion=Locked
  twistLimitDegrees=180
 }
 {
  bone1Name=dyn_lower_panel_R
  bone2Name=lower_panel_R
  swing1LimitDegrees=45
  swing1Motion=Locked
  swing2LimitDegrees=45
  swing2Motion=Locked
  twistLimitDegrees=180
 }
 {
  bone1Name=dyn_upper_panel_01_L
  bone2Name=upper_panel_01_L
  swing1LimitDegrees=45
  swing1Motion=Locked
  swing2LimitDegrees=45
  swing2Motion=Locked
  twistLimitDegrees=180
 }
 {
  bone1Name=dyn_upper_panel_02_L
  bone2Name=upper_panel_02_L
  swing1LimitDegrees=45
  swing1Motion=Locked
  swing2LimitDegrees=45
  swing2Motion=Locked
  twistLimitDegrees=180
 }
 {
  bone1Name=dyn_upper_panel_01_R
  bone2Name=upper_panel_01_R
  swing1LimitDegrees=45
  swing1Motion=Locked
  swing2LimitDegrees=45
  swing2Motion=Locked
  twistLimitDegrees=180
 }
 {
  bone1Name=dyn_upper_panel_02_R
  bone2Name=upper_panel_02_R
  swing1LimitDegrees=45
  swing1Motion=Locked
  swing2LimitDegrees=45
  swing2Motion=Locked
  twistLimitDegrees=180
 }
 {
  bone1Name=lower_panel_L
  bone2Name=lower_axis
  swing1LimitDegrees=45
  swing1Motion=Free
  swing2LimitDegrees=45
  swing2Motion=Free
  twistLimitDegrees=45
  twistMotion=Free
 }
 {
  bone1Name=lower_panel_R
  bone2Name=lower_axis
  swing1LimitDegrees=45
  swing1Motion=Free
  swing2LimitDegrees=45
  swing2Motion=Free
  twistLimitDegrees=45
  twistMotion=Free
 }
 {
  bone1Name=upper_panel_01_L
  bone2Name=upper_axis_L
  swing1LimitDegrees=45
  swing1Motion=Free
  swing2LimitDegrees=45
  swing2Motion=Free
  twistLimitDegrees=45
  twistMotion=Free
 }
 {
  bone1Name=upper_panel_02_L
  bone2Name=upper_axis_L
  swing1LimitDegrees=45
  swing1Motion=Free
  swing2LimitDegrees=45
  swing2Motion=Free
  twistLimitDegrees=45
  twistMotion=Free
 }
 {
  bone1Name=upper_panel_01_R
  bone2Name=upper_axis_R
  swing1LimitDegrees=45
  swing1Motion=Free
  swing2LimitDegrees=45
  swing2Motion=Free
  twistLimitDegrees=45
  twistMotion=Free
 }
 {
  bone1Name=upper_panel_02_R
  bone2Name=upper_axis_R
  swing1LimitDegrees=45
  swing1Motion=Free
  swing2LimitDegrees=45
  swing2Motion=Free
  twistLimitDegrees=45
  twistMotion=Free
 }
]
