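# Animation Blueprint for the TungStan pickaxe mesh that adds physics-driven secondary motion to the weapon.
# The input pose passes through DefaultSlot, an additive yaw rotation on the two-handed handle bone, then a RigidBody node.
# The simulation runs in the handle bone's space with an override physics asset and gravity.
# Each update inverts the emote_BuiltIn curve from a part mesh's anim instance into the simulation's alpha, fading physics out during built-in emotes.
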AnimBlueprint /BRCosmetics/Weapons/FORT_Melee/Pickaxe_TungStan/Meshes/TungStan_Axe_AnimBP
target_skeleton=&/TungStan_Axe_Skeleton
graphs [
 AnimGraph {nodes [
   n1 = Component To Local(ComponentPose=Local To Component(LocalPose=Input Pose()))
   n2 = "Transform (Modify) Bone - Bone: Melee_TwoHanded_Handle"!(ComponentPose=Local To Component(LocalPose=Slot 'DefaultSlot'(Source=@n1)), Rotation="0,45,0") -> (Pose) {rotationMode=BMM_Additive, alphaBoolBlend {blendOption=Linear}, boneToModify {boneName=Melee_TwoHanded_Handle}, rotation {yaw=45}}
   n3 = RigidBody(ComponentPose=@n2, Alpha=Get Anim Alpha() -> (Anim Alpha:double)) {
    bClampLinearTranslationLimitToRefPose=true
    bOverrideWorldGravity=true
    overridePhysicsAsset=&/TungStan_Axe_Physics
    simulationSpace=BaseBoneSpace
    alphaBoolBlend {blendInTime=0.7, blendOption=Linear}
    baseBoneRef {boneName=Melee_TwoHanded_Handle}
    componentAppliedLinearAccClamp=<3500 3500 3500>
    componentLinearAccScale=<0.55 0.5 0.55>
    componentLinearVelScale=<0.55 0.5 0.55>
    overrideWorldGravity {z=-980}
   }
   Output Pose(Result=Component To Local(ComponentPose=@n3)) {layerGroup=None}
  ]}
 EventGraph {nodes [
   Event BlueprintUpdateAnimation()
   n1 = GetAnimInstance(self=GetPartSkeletalMeshComponent())
   > Is Valid(InputObject=@n1)
   > Set Anim Alpha(Anim Alpha:double=MapRangeClamped(Value=GetCurveValue(self=@n1, CurveName=emote_BuiltIn), InRangeA=0, InRangeB=1, OutRangeA=1, OutRangeB=0)) -> (Output_Get:double)
  ]}
]
variables [
 Anim Alpha {type=number}
 notPlayingEquipingAnim {type=boolean}
]
